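# Sound cue for the initial launch of the version 2 witch broom gadget in Athena.
# A Random node with equal weights picks one of three launch wave variants and is excluded from branch culling.
# It borrows glider audio setup, using the glider sound class, glider open/close attenuation and glider open/close concurrency, with a slight volume boost.
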SoundCue /FNGameplayCues/Sounds/Fort_Gadgets/Athena/Broomstick/v2/WitchBroom_v2_InitialLaunch_Cue
root=Random_0
nodes [
 {id=Random_0, children [WavePlayer_0,WavePlayer_4,WavePlayer_5], properties {bShouldExcludeFromBranchCulling=true, weights [1,1,1]}}
 {id=WavePlayer_0, properties {soundWaveAssetPtr=&/WitchBroom_v2_InitialLaunch_01}}
 {id=WavePlayer_4, properties {soundWaveAssetPtr=&/WitchBroom_v2_InitialLaunch_02}}
 {id=WavePlayer_5, properties {soundWaveAssetPtr=&/WitchBroom_v2_InitialLaunch_03}}
]
properties {
 attenuationSettings=/Game/Sounds/Attenuation/Fort_Player_Glider_OpenClose_Athena_Att
 maxDistance=10000
 soundClassObject=/Game/Sounds/FortSoundClassesAndModes/Fort_Gliders
 volumeMultiplier=1.2
 concurrencySet [&Athena_GliderOpenClose_Conc]
}
